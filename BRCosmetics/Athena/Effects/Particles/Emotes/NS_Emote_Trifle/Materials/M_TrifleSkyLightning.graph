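# Translucent unlit material for Niagara mesh particles that draws the sky lightning bolts of the Trifle emote.
# SplineThicken sets bolt width, while two layered noise functions displace vertices, seeded by a dynamic parameter, with the finer layer animated over time.
# A gradient mask moves along V with particle age and a random offset, combining with brightness oscillation and sphere masks to shape width, emissive and opacity, tinted by particle color.
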Material /BRCosmetics/Athena/Effects/Particles/Emotes/NS_Emote_Trifle/Materials/M_TrifleSkyLightning
graph {
 properties {
  bUsedWithNiagaraMeshParticles=true
  blendMode=BLEND_Translucent
  pixelDepthOffsetMode=PDOM_Legacy
  shadingModel=MSM_Unlit
  usedShadingModels=Unlit
 }
 nodes [
  n1 = MaterialFunctionCall!(In) {materialFunction=/BRCosmetics/Weapons/FORT_Melee/Pickaxe_Parfait_Female/FX/Materials/Functions/MF_TrifleParfait_BrightnessOscillation}
  n2 = ParticleRelativeTime()
  n3 = LinearInterpolate(Alpha=ParticleRandom()) {constB=0.2}
  n4 = MaterialFunctionCall!(
   Gradient=ComponentMask(None=TextureCoordinate()) {g=true}
   Mask Position=LinearInterpolate(Alpha=Add(A=@n2, B=@n3))
   Mask Size=DynamicParameter() {paramNames [RandomTime, MaskSize, LightningWidth, Param4]}.MaskSize
  ) {materialFunction=/Game/Effects/Fort_Effects/MaterialFunctions/MF_GradientMask}
  n5 = Power(Base=@n4.Sinusoidal) {constExponent=5}
  n6 = SphereMask(A=TextureCoordinate() {uTiling=0}, B=Constant2Vector() {g=0.55}, Radius=Constant() {r=0.525}, Hardness=Constant() {r=0.7})
  n7 = LinearInterpolate(
   B=DynamicParameter() {paramNames [RandomTime, MaskSize, LightningWidth, Param4]}.LightningWidth
   Alpha=Multiply(A=Multiply(A=@n1.Brightness, B=@n5), B=@n6)
  )
  n8 = DynamicParameter() {paramNames [RandomTime, MaskSize, LightningWidth, Param4]}
  n9 = MaterialFunctionCall!(Noise Intensity=Constant() {r=7}, Noise UV Scale=Constant() {r=0.05}, Time=@n8.RandomTime, Noise Axis Rotation) {materialFunction=/BRCosmetics/Weapons/FORT_Melee/Pickaxe_Parfait_Female/FX/Materials/Functions/MF_TrifleParfait_Noise}
  n10 = Time()
  n11 = MaterialFunctionCall(
   Noise Intensity=Multiply(A=@n9.Noise Width Multiplier, B=Constant() {r=2})
   Noise UV Scale=Constant() {r=0.3}
   Time=Add(A=@n8.RandomTime, B=Multiply(A=@n10) {constB=0.5})
  ) {materialFunction=&MF_TrifleParfait_Noise}
  Add(
   A=MaterialFunctionCall(WidthBase=@n7, WidthTip=@n7) {materialFunction=&SplineThicken}.WorldPositionOffset
   B=Multiply(A=@n6, B=Add(A=@n9.Noise WPO, B=@n11.Noise WPO))
  )
  n12 = ParticleColor()
  n13 = SphereMask(A=TextureCoordinate() {vTiling=0}, B=Constant2Vector() {r=0.5}, Radius=Constant() {r=0.5}, Hardness=Constant())
  n14 = Saturate(None=Multiply(A=Multiply(A=@n13, B=@n1.Brightness), B=Power(Base=@n4.Sinusoidal) {constExponent=5}))
  MaterialFunctionCall(Emissive=Multiply(A=@n12.RGB, B=@n14)) {materialFunction=&MF_FX_EmissiveAdjustments}
  Multiply(A=@n12.A, B=@n14)
 ]
}
