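# Material function for the Trifle emote's mesh-particle flare, computing an offset that re-forms the mesh as a camera-facing shape around its particle position.
# Centred, V-flipped UVs are scaled by each point's distance from the particle centre, then transformed from camera space.
# Subtracting world position yields the offset, plus a FlareOffset input transformed from instance space.
# The length clamp near one leaves typical 0–1 UV vectors effectively unnormalized.
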MaterialFunction /BRCosmetics/Athena/Effects/Particles/Emotes/NS_Emote_Trifle/Functions/MF_Trifle_Meshparticle_Flare
graph {nodes [
  n1 = TextureCoordinate()
  n2 = ComponentMask(None=@n1) {g=true}
  n3 = Subtract(A=AppendVector(A=ComponentMask(None=@n1) {r=true}, B=OneMinus(None=@n2))) {constB=0.5}
  n4 = SquareRoot(None=DotProduct(A=@n3, B=@n3))
  n5 = WorldPosition()
  n6 = ParticlePositionWS()
  n7 = Multiply(A=Divide(A=@n3, B=Max(A=@n4) {constB=0.9999}), B=Distance(A=@n5.XYZ, B=@n6))
  n8 = Constant()
  n9 = Add(A=Transform(None=AppendVector(A=@n7, B=@n8)) {transformSourceType=TRANSFORMSOURCE_Camera}, B=@n6)
  n10 = FunctionInput(Preview=Constant3Vector() {constant=<rgba 0 0 1000 0>}) {bUsePreviewValueAsDefault=true, inputName=FlareOffset}
  FunctionOutput(None=Add(A=Subtract(A=@n9, B=@n5.XYZ), B=Transform(None=@n10) {transformSourceType=TRANSFORMSOURCE_Instance}))
 ]}
