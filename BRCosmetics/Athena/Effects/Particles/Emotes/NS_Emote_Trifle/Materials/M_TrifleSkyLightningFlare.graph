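# Translucent unlit mesh-particle material for a blue lightning flare in the Trifle emote's Niagara effect.
# A center-scaled, per-particle randomly rotated spikeball flare flickers via panned noise sampled with zero-tiled UVs, plus a soft radial glow.
# The combined mask drives tinted emissive, and opacity after subtracting the Erode dynamic parameter and multiplying by particle alpha.
# It also calls a Trifle mesh-particle flare offset function.
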Material /BRCosmetics/Athena/Effects/Particles/Emotes/NS_Emote_Trifle/Materials/M_TrifleSkyLightningFlare
graph {
 properties {
  bUsedWithNiagaraMeshParticles=true
  blendMode=BLEND_Translucent
  pixelDepthOffsetMode=PDOM_Legacy
  shadingModel=MSM_Unlit
  usedShadingModels=Unlit
 }
 nodes [
  MaterialFunctionCall!(FlareOffset) {materialFunction=/BRCosmetics/Athena/Effects/Particles/Emotes/NS_Emote_Trifle/Functions/MF_Trifle_Meshparticle_Flare}
  n1 = TextureCoordinate()
  n2 = Constant() {r=1.2}
  n3 = ParticleRandom()
  n4 = TextureSample(UVs=MaterialFunctionCall(
    UVs=MaterialFunctionCall(UVs=@n1, Texture Scale=@n2) {materialFunction=&ScaleUVsByCenter}.UVs
    "Rotation Angle (0-1)"=LinearInterpolate(Alpha=@n3)
   ) {materialFunction=&CustomRotator}.Rotated Values) {texture=/Game/Effects/Fort_Effects/Textures/Flares/T_Flare_Spikeball}
  n5 = TextureCoordinate() {uTiling=0, vTiling=0}
  n6 = LinearInterpolate(Alpha=TextureSample(UVs=Panner(Coordinate=@n5) {speedX=1.11, speedY=1.11}) {texture=&T_NoisePack_01}.R) {constA=0.5}
  n7 = Power(Base=SphereMask(A=TextureCoordinate(), B=Constant() {r=0.5}, Radius=Constant() {r=0.5}, Hardness=Constant())) {constExponent=4}
  n8 = Saturate(None=Add(A=Multiply(A=@n4.RGB, B=@n6), B=Multiply(A=@n7) {constB=0.15}))
  MaterialFunctionCall(Emissive=Multiply(A=Constant3Vector() {constant=<rgba 0.4308 1.9075 6 0>}, B=@n8)) {materialFunction=&MF_FX_EmissiveAdjustments}
  Multiply(A=ParticleColor().A, B=Subtract(A=@n8, B=DynamicParameter() {parameterIndex=1, paramNames [Erode,Param2,Param3,Param4]}.Erode))
 ]
}
