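# Niagara system for lightning effects on the hammer in the Trifle emote.
# A Glow emitter (mesh, sprite, light), two mesh lightning emitters, sprite sparks gated by PowersActive, and a ring flash burst combine.
# System Flicker, timer and time-of-day values modulate brightness, alpha and light radius.
# LightningMeshes picks meshes in LightningOrder, randomly animates traveling or static bolts, and uses the timeline to shorten its loop interval.
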NiagaraSystem /BRCosmetics/Athena/Effects/Particles/Emotes/NS_Emote_Trifle/NS_Emote_Trifle_HammerLightning
dependencies {
 usedDataInterfaces [
  &NiagaraDataInterfaceCurve
  &NiagaraDataInterfaceArrayFloat3
  &NiagaraDataInterfaceColorCurve
  &NiagaraDataInterfaceArrayFloat
 ]
 usedDynamicInputs [&RandomVector, &FNTimeOfDayScalar]
 usedModules [
  &SystemState
  &FN_Timer
  &self:SystemSpawnScript.SystemScriptSource.SystemScriptGraph.NiagaraNodeAssignment_0.SetVariables_#1
  &Timeline
  &self:SystemSpawnScript.SystemScriptSource.SystemScriptGraph.NiagaraNodeAssignment_1.SetVariables_#2
 ]
}
emitters {
 Glow {
  emitterUpdateScript [
   &EmitterState(
    Inactive Response=::NewEnumerator1
    Life Cycle Mode=::NewEnumerator1
    Loop Behavior=::NewEnumerator1
    Loop Duration Mode=::NewEnumerator1
    UseLoopDelay=false
   )
   &SpawnBurst_Instantaneous(Use Spawn Probability=false, UseLoopCountLimit=false)
  ]
  particleSpawnScript [
   &InitializeParticle(
    Color=<rgba 0.0567 0.3207 1 1>
    Position Offset=<0 0 26>
    Sprite Size Mode=::NewEnumerator3
    Uniform Sprite Size=250
    UsePositionOffset=true
    Write Material Random=true
   )
   /SetVariables() {isSetParameters=true}
   /SetVariables() {isSetParameters=true}
  ]
  particleUpdateScript [
   &ParticleState(Kill Particles When Lifetime Has Elapsed=false)
   /SetVariables(
    Particles.FlashSize:Vector2f=&Multiply_Vector2D_ByFloat(
     Float=&FloatFromCurve(
      FloatCurve=DataInterface() {propertyValues {Curve {keys [{tangentMode=RCTM_User, time=0, value=0}, {tangentMode=RCTM_User, time=1.5, value=1.5}]}}}
      CurveIndex=$Emitter.LoopedAge
     )
     Vector2D=$Particles.SpriteSize
    )
    Particles.LightColor=&MakeLinearColorFromVectorAndFloat(
     "Vector (RGB)"=&Multiply_VectorByFloat(Vector=&MakeVectorFromLinearColorRGB(Color=$Particles.Color), Float=15)
     "Float (Alpha)"=&Multiply_Float(A=$System.FN_Timer.Value, B=&Lerp_Float(A=0.75, B=1, Alpha=$System.Flicker))
    )
    Particles.LightExponent:NiagaraFloat=&Lerp_Float(A=1.666, B=2.333, Alpha=$System.Flicker)
    Particles.LightRadius=&Lerp_Float(A=100, B=150, Alpha=$System.Flicker)
    Particles.MeshColor:LinearColor=&MakeLinearColorFromVectorAndFloat(
     "Vector (RGB)"=&Multiply_VectorByFloat(
      Vector=&MakeVectorFromLinearColorRGB(Color=$Particles.Color)
      Float=&Multiply_Float(A=$System.TimeOfDayBrightnessMultiplier, B=11)
     )
     "Float (Alpha)"=&Multiply_Float(A=$System.FN_Timer.Value, B=&Lerp_Float(A=0.4, B=1, Alpha=$System.Flicker))
    )
    Particles.SpriteColor=&MakeLinearColorFromVectorAndFloat(
     "Vector (RGB)"=&Multiply_VectorByFloat(
      Vector=&Multiply_VectorByFloat(
       Vector=&MakeVectorFromLinearColorRGB(Color=$Particles.Initial.Color)
       Float=$System.TimeOfDayBrightnessMultiplier
      )
      Float=0.6
     )
     "Float (Alpha)"=&Multiply_Float(A=$System.FN_Timer.Value, B=&Lerp_Float(A=0.6, B=1, Alpha=$System.Flicker))
    )
   ) {isSetParameters=true}
  ]
  rendererClasses [
   &NiagaraMeshRendererProperties
   &NiagaraLightRendererProperties
   &NiagaraSpriteRendererProperties
  ]
  renderers [
   {rendererClass=&NiagaraMeshRendererProperties, rendererIndex=0}
   {rendererClass=&NiagaraLightRendererProperties, rendererIndex=1}
   {rendererClass=&NiagaraSpriteRendererProperties, rendererIndex=2}
  ]
 }
 Lightning {
  emitterUpdateScript [
   &EmitterState(Life Cycle Mode=::NewEnumerator1, Loop Behavior=::NewEnumerator1, Loop Duration=1.1, UseLoopDelay=true)
   &SpawnRate(SpawnRate=9, Use Spawn Probability=false)
   /SetVariables(Emitter.GlobalOffsetIntensity=4) {isSetParameters=true}
  ]
  particleSpawnScript [
   &InitializeParticle(
    Color=<rgba 0 1.3826 5 1>
    Lifetime=&Lerp_Float(
     A=&RandomRangeFloat(Minimum=0.5, Maximum=0.75)
     B=&RandomRangeFloat(Minimum=1.5, Maximum=2)
     Alpha=&FloatFromCurve(
      FloatCurve=DataInterface() {propertyValues {Curve {keys [
          {arriveTangent=9.0725, interpMode=RCIM_Cubic, leaveTangent=9.0725, tangentMode=RCTM_User, time=0, value=0}
          {arriveTangent=-2.5249, interpMode=RCIM_Cubic, leaveTangent=-2.5249, tangentMode=RCTM_User, time=0.3, value=1}
          {arriveTangent=-0.0405, interpMode=RCIM_Cubic, leaveTangent=-0.0405, tangentMode=RCTM_User, time=1, value=0}
         ]}}}
      CurveIndex=$Emitter.LoopedAge
     )
    )
    Mesh Scale Mode=::NewEnumerator1
    Mesh Uniform Scale Max=0.175
    Mesh Uniform Scale Min=0.125
    Position Offset=&RandomRangeVector(Minimum=<-2 -3 25>, Maximum=<2 3 25>)
    Position Offset Coordinate Space=::Local
    UsePositionOffset=true
    Write Material Random=true
   )
   &DynamicMaterialParameters(
    Index 0 Param 1=&RandomRangeFloat(Minimum=0, Maximum=1)
    Index 0 Param 2=&RandomRangeFloat(Minimum=2.7703, Maximum=3.5522)
    Index 0 Param 3=&RandomRangeFloat(Minimum=3, Maximum=4)
    Module.Param0WriteEnabled=true
    Module.Param1WriteEnabled=true
    Module.Param2WriteEnabled=true
    Module.Param3WriteEnabled=false
   )
   &InitialMeshOrientation(
    Rotation=&RandomRangeVector(Minimum=<0.25 0.25 0>, Maximum=<-0.25 -0.25 0>)
    Rotation Coordinate Space=::NewEnumerator1
    Use Rotation Vector=true
   )
  ]
  particleUpdateScript [
   &ParticleState()
   &ScaleColor(
    Scale Alpha=$System.TimeOfDayBrightnessMultiplier
    Scale RGB=&Multiply_VectorByFloat(
     Vector=<1 1 1>
     Float=&FloatFromCurve(FloatCurve=DataInterface() {propertyValues {Curve {keys [
          {arriveTangent=1320.8182, interpMode=RCIM_Cubic, leaveTangent=1320.8182, tangentMode=RCTM_User, time=0.3, value=11}
          {interpMode=RCIM_Cubic, tangentMode=RCTM_User, time=0.4, value=88}
          {tangentMode=RCTM_User, time=0.475, value=11}
         ]}}})
    )
    ScaleA=true
    ScaleRGB=true
   )
   &DynamicMaterialParameters(
    Index 1 Param 1=&FloatFromCurve(FloatCurve=DataInterface() {propertyValues {Curve {keys [
         {interpMode=RCIM_Cubic, tangentMode=RCTM_User, time=0.2, value=0}
         {interpMode=RCIM_Cubic, tangentMode=RCTM_User, time=0.3, value=1}
         {tangentMode=RCTM_User, time=0.4, value=0}
        ]}}})
    Module.Param0WriteEnabled1=true
    Module.Param1WriteEnabled1=true
    Module.Param2WriteEnabled1=true
    Module.Param3WriteEnabled1=true
    Write Parameter Index 0=false
    Write Parameter Index 1=true
   ) {module=DynamicMaterialParameters001}
   &Drag(Use Linear Drag=true, Use Rotational Drag=false)
   &SolveForcesAndVelocity(Clamp Velocity=false, Limit Acceleration=false)
  ]
  rendererClasses [&NiagaraMeshRendererProperties]
  renderers [{rendererClass=&NiagaraMeshRendererProperties, rendererIndex=0}]
 }
 LightningMeshes {
  emitterUpdateScript [
   &EmitterState(
    Inactive Response=::NewEnumerator2
    Life Cycle Mode=::NewEnumerator1
    Loop Duration=&Lerp_Float(A=0.3, B=&RandomRangeFloat(Minimum=0.0666, Maximum=0.1666), Alpha=$System.Timeline.Float)
    Recalculate Duration Each Loop=true
    UseLoopDelay=false
   )
   &SpawnBurst_Instantaneous(Use Spawn Probability=false, UseLoopCountLimit=false)
  ]
  particleSpawnScript [
   /SetVariables(
    Particles.MeshIndex=&SelectIntFromArray(Direct Array Index=$Emitter.LoopCount, Direct Array Mode=::NewEnumerator1, Int Selection Array=$User.LightningOrder)
    Particles.bIsTravelingBolt:NiagaraBool=&RandomBool()
   ) {isSetParameters=true}
   &InitializeParticle(
    Color=<rgba 0 48.0999 222 1>
    Lifetime=&MakeCustomFloatFromBool(Bool=$Particles.bIsTravelingBolt, True Float=0.5, False Float=0.4)
    UsePositionOffset=false
    Write Material Random=true
   )
   &DynamicMaterialParameters(
    Index 0 Param 1=1.333
    Index 0 Param 2=&MakeCustomFloatFromBool(Bool=$Particles.bIsTravelingBolt, True Float=0.75, False Float=3)
    Index 0 Param 4=0.5
    Module.Param0WriteEnabled=true
    Module.Param1WriteEnabled=true
    Module.Param2WriteEnabled=true
    Module.Param3WriteEnabled=true
   )
   &InitialMeshOrientation(
    Mesh Orientation Mode=::NewEnumerator7
    Rotation=&SelectVectorFromArray(
     Direct Array Index=/Niagara/DynamicInputs/Divide/Divide_Int(A:NiagaraInt32=$Emitter.LoopCount, B:NiagaraInt32=4)
     Direct Array Mode=::NewEnumerator1
     Vector Selection Array=DataInterface() {propertyValues {FloatData [<0 0 0.5>, <0 0 0>]}}
    )
    Use Rotation Vector=true
   )
  ]
  particleUpdateScript [
   &ParticleState()
   &DynamicMaterialParameters(
    Index 0 Param 1=&MakeCustomFloatFromBool(
     Bool=$Particles.bIsTravelingBolt
     True Float=1.333
     False Float=&FloatFromCurve(FloatCurve=DataInterface() {propertyValues {Curve {keys [<at 0.666 1>,<at 1 0>]}}}, Scale Curve=1.333)
    )
    Index 0 Param 3=&MakeCustomFloatFromBool(
     Bool=$Particles.bIsTravelingBolt
     True Float=$Particles.NormalizedAge
     False Float=&FloatFromCurve(FloatCurve=DataInterface() {propertyValues {Curve {keys [<at 0 0>,<at 0.333 0.5>]}}})
    )
    Index 0 Param 4=&MakeCustomFloatFromBool(
     Bool=$Particles.bIsTravelingBolt
     True Float=0.5
     False Float=&FloatFromCurve(FloatCurve=DataInterface() {propertyValues {Curve {keys [{tangentMode=RCTM_User, time=0, value=0.333}, {tangentMode=RCTM_User, time=1, value=0.666}]}}})
    )
    Module.Param0WriteEnabled=true
    Module.Param1WriteEnabled=false
    Module.Param2WriteEnabled=true
    Module.Param3WriteEnabled=true
   ) {module=DynamicMaterialParameters001}
   /SetVariables() {enabled=false, isSetParameters=true}
   &ScaleColor(
    Scale RGB=&Multiply_Vector(
     A=&VectorFromFloat(Value=&Lerp_Float(A=0.133, B=1, Alpha=$System.Flicker))
     B=&VectorFromFloat(Value=$System.TimeOfDayBrightnessMultiplier)
    )
    ScaleA=true
    ScaleRGB=true
   )
  ]
  rendererClasses [&NiagaraMeshRendererProperties]
  renderers [{rendererClass=&NiagaraMeshRendererProperties, rendererIndex=0}]
 }
 LightningSprites {
  emitterUpdateScript [
   &EmitterState()
   &SpawnRate(
    Spawn Probability=0.5
    SpawnRate=&MakeCustomFloatFromBool(Bool=$User.PowersActive, True Float=16, False Float=0)
    Use Spawn Probability=true
   )
  ]
  particleSpawnScript [
   &InitializeParticle(
    Color=<rgba 0 1.3 6 1>
    Lifetime Max=0.166
    Lifetime Min=0.133
    Lifetime Mode=::NewEnumerator1
    Position Offset=<0 0 25>
    Sprite Rotation Mode=::NewEnumerator2
    Sprite Size Mode=::NewEnumerator1
    Sprite UV Mode=::NewEnumerator3
    Uniform Sprite Size Max=29
    Uniform Sprite Size Min=21
    UsePositionOffset=true
    Write Material Random=true
   ) {module=InitializeParticle001}
   /SetVariables(Particles.SubImageIndex=&UniformRangedFloat(Minimum=0, Maximum=3)) {isSetParameters=true}
   &CameraOffset(Camera Offset Amount=9)
   &ShapeLocation(Box Size=<22 11 11>, MaskSpawnGroup=false, Override Seed=false, Shape Primitive=::NewEnumerator2, Surface Only=false, UseOwnerScale=true)
  ]
  particleUpdateScript [
   &DynamicMaterialParameters(
    Index 0 Param 1=&RandomRangeFloat(Minimum=0, Maximum=1)
    Index 0 Param 2=144
    Index 0 Param 3=&FloatFromCurve(FloatCurve=DataInterface() {propertyValues {Curve {keys [
         {arriveTangent=-19.4071, interpMode=RCIM_Cubic, leaveTangent=-19.4071, tangentMode=RCTM_User, time=0, value=1}
         {interpMode=RCIM_Cubic, tangentMode=RCTM_Break, time=0.1, value=0}
         {arriveTangent=3.3801, interpMode=RCIM_Cubic, leaveTangent=3.3801, tangentMode=RCTM_User, time=1, value=1}
        ]}}})
    Module.Param0WriteEnabled=true
    Module.Param1WriteEnabled=true
    Module.Param2WriteEnabled=true
    Module.Param3WriteEnabled=false
   )
   &ParticleState()
   &ScaleColor(
    Scale Alpha=&FloatFromCurve(FloatCurve=DataInterface() {propertyValues {Curve {keys [<at 0.75 1 RCIM_Cubic>, {arriveTangent=-10.3917, interpMode=RCIM_Cubic, leaveTangent=-10.3917, tangentMode=RCTM_User, time=1, value=0}]}}})
    ScaleA=true
    ScaleRGB=false
   ) {module=ScaleColor002}
   &ScaleSpriteSize(
    Uniform Curve Scale=&Lerp_Float(A=0.9, B=1, Alpha=$System.Flicker)
    Uniform Curve Sprite Scale=DataInterface() {propertyValues {Curve {keys [{arriveTangent=1.4996, time=0, value=0.9}, <at 0.2 1>, <at 0.7 0.98>, {leaveTangent=-0.2857, time=1, value=1}]}}}
   )
   &ScaleColor(Scale Alpha=$System.TimeOfDayBrightnessMultiplier, ScaleA=true, ScaleRGB=true) {module=ScaleColor001}
  ]
  rendererClasses [&NiagaraSpriteRendererProperties]
  renderers [{rendererClass=&NiagaraSpriteRendererProperties}]
 }
 Ring {
  emitterUpdateScript [
   &EmitterState(
    Enable Distance Culling=true
    Life Cycle Mode=::NewEnumerator1
    Loop Behavior=::NewEnumerator1
    Loop Delay=0.1
    Loop Duration=2
    Max Distance Response=::KillImmediately
    MaxDistance=3500
    Module.bUseMaxDistance=true
    Module.bUseMinDistance=false
    Scalability Mode=::NewEnumerator1
    UseLoopDelay=false
   )
   &SpawnBurst_Instantaneous(Spawn Count=1, Spawn Time=1.15, Use Spawn Probability=false)
  ]
  particleSpawnScript [
   &InitializeParticle(
    Color=<rgba 0.5782 0.7687 1 0.5>
    Lifetime=0.3
    Position Offset=<0 0 25>
    Sprite Rotation Mode=::NewEnumerator2
    Sprite Size Mode=::NewEnumerator3
    Uniform Sprite Size=150
    UsePositionOffset=true
    Write Material Random=true
   )
   &CameraOffset(Camera Offset Amount=35)
   /SetVariables(Particles.FlashColorInitial:LinearColor=<rgba 0.1439 1.8866 5 1>, Particles.FlashInitial Size:Vector2f=<550 300>) {isSetParameters=true}
  ]
  particleUpdateScript [
   &ParticleState()
   &ScaleColor(
    Scale Alpha=$System.TimeOfDayBrightnessMultiplier
    Scale RGB=&Multiply_VectorByFloat(
     Vector=<1 1 1>
     Float=&FloatFromCurve(FloatCurve=DataInterface() {propertyValues {Curve {keys [
          {interpMode=RCIM_Cubic, tangentMode=RCTM_User, time=0, value=0}
          {interpMode=RCIM_Cubic, tangentMode=RCTM_User, time=0.2, value=0.95}
          {tangentMode=RCTM_User, time=1, value=0}
         ]}}})
    )
    ScaleA=true
    ScaleRGB=true
   )
   &ScaleSpriteSize(Scale Factor=&Vector2DFromFloat(Value=&FloatFromCurve(FloatCurve=DataInterface() {propertyValues {Curve {keys [
          {arriveTangent=2.5154, interpMode=RCIM_Cubic, leaveTangent=2.5154, tangentMode=RCTM_User, time=0, value=0}
          {arriveTangent=1, interpMode=RCIM_Cubic, leaveTangent=1, time=0.1, value=0.6}
          <at 1 1>
         ]}}})))
   /SetVariables(
    Particles.FlashCOlor:LinearColor=&MakeLinearColorFromVectorAndFloat(
     "Vector (RGB)"=&MakeVectorFromLinearColorRGB(Color=$Particles.FlashColorInitial)
     "Float (Alpha)"=&FloatFromCurve(FloatCurve=DataInterface() {propertyValues {Curve {keys [
          {arriveTangent=9.0725, interpMode=RCIM_Cubic, leaveTangent=9.0725, tangentMode=RCTM_User, time=0, value=0}
          {arriveTangent=-2.5249, interpMode=RCIM_Cubic, leaveTangent=-2.5249, tangentMode=RCTM_User, time=0.1408, value=0.9847}
          {arriveTangent=-0.0405, interpMode=RCIM_Cubic, leaveTangent=-0.0405, tangentMode=RCTM_User, time=1, value=0}
         ]}}})
    )
    Particles.FlashSize:Vector2f=&Multiply_Vector2D_ByFloat(
     Float=&FloatFromCurve(FloatCurve=DataInterface() {propertyValues {Curve {keys [
          {arriveTangent=4.1785, interpMode=RCIM_Cubic, leaveTangent=4.1785, tangentMode=RCTM_User, time=0, value=0}
          {arriveTangent=0.9995, interpMode=RCIM_Cubic, leaveTangent=0.9995, tangentMode=RCTM_User, time=0.2223, value=0.6825}
          {arriveTangent=-0.0405, interpMode=RCIM_Cubic, leaveTangent=-0.0405, tangentMode=RCTM_User, time=1, value=1}
         ]}}})
     Vector2D=$Particles.FlashInitial Size
    )
   ) {isSetParameters=true}
  ]
  rendererClasses [&NiagaraSpriteRendererProperties]
  renderers [
   {rendererClass=&NiagaraSpriteRendererProperties}
   {rendererClass=&NiagaraSpriteRendererProperties, rendererIndex=1}
  ]
 }
}
summary {
 systemName=NS_Emote_Trifle_HammerLightning
 emitters [
  {emitterName=LightningSprites, rendererClasses [&NiagaraSpriteRendererProperties]}
  {
   emitterName=Glow
   rendererClasses [
    &NiagaraMeshRendererProperties
    &NiagaraLightRendererProperties
    &NiagaraSpriteRendererProperties
   ]
  }
  {emitterName=LightningMeshes, rendererClasses [&NiagaraMeshRendererProperties]}
  {emitterName=Ring, rendererClasses [&NiagaraSpriteRendererProperties]}
  {emitterName=Lightning, rendererClasses [&NiagaraMeshRendererProperties]}
 ]
 userVariables [
  User.bIsFrontEnd {type=NiagaraBool}
  User.FadeOut {type=NiagaraFloat, defaultValue=0, description=""}
  User.FlickerSpeed {type=NiagaraFloat, defaultValue=6.666, description=""}
  User.LightningOrder {
   type=NiagaraDataInterfaceArrayInt32
   description=""
   defaultValue {
    struct=NiagaraExt_VariableValue_DataInterface
    value {
     dataInterface=&self:NiagaraDataInterfaceArrayInt32_0
     dataInterfaceClass=&NiagaraDataInterfaceArrayInt32
    }
   }
  }
  User.LightningSpriteDepthFade {type=NiagaraFloat, defaultValue=44, description=""}
  User.PowersActive {type=NiagaraBool, defaultValue=true}
  User.RibbonIDOffset {type=NiagaraInt32}
  User.SkeletalMesh {
   type=NiagaraDataInterfaceSkeletalMesh
   defaultValue {
    struct=NiagaraExt_VariableValue_DataInterface
    value {dataInterface=&self:NiagaraDataInterfaceSkeletalMesh_0}
   }
  }
 ]
}
